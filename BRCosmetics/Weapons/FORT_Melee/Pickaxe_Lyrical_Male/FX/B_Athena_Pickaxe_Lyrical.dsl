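; Lyrical pickaxe weapon Blueprint that adds a visual reaction to weak-point hits on top of the parent impact FX.
; Impact events pass the physical surface to a check against the weak-point surface list, with a debug print left in.
; A DoOnce gate leads to timelines that drive the idle VFX component's Point and Point_Light parameters, with retriggerable delays staging the follow-up curves.
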(event FortWeapon|EventOnPlayImpactFX
  (params (HitResult :type Hit%20Result%20Structure%20%28by%20ref%29) (ImpactPhysicalSurface :type EPhysicalSurface%20Enum :default "SurfaceType_Default") (SpawnedPSC :type FXSystem%20Component%20Object%20Reference))
  (|CallParentFunction HitResult ImpactPhysicalSurface SpawnedPSC)
  (CallFunction|ImpactFXPoint ImpactPhysicalSurface))

(event Custom|ImpactFX_Point
  (params (Impact%20Physical%20Surface :type EPhysicalSurface%20Enum :default "SurfaceType_Default"))
  (CallFunction|WeakPoint Impact%20Physical%20Surface))

(event Custom|WeakPoint
  (params (NewParam :type EPhysicalSurface%20Enum :default "SurfaceType_Default"))
  (bind _returnvalue (CallFunction|GetIdleVFXComponent))
  (bind _output_get (Variables|Default|SetCurrent%20Impact%20Physical%20Surface NewParam))
  (Variables|Default|SetNewVar_0 (Utilities|Array|ContainsItem (Variables|Default|GetWeakPoint%20Physical%20Surface) _output_get))
  (if (Variables|Default|GetNewVar_0)
    (Development|PrintString (Utilities|GetDisplayName _returnvalue))
    (Utilities|FlowControl|DoOnce false
      (:Completed
        (bind (_direction _newtrack_0) (Timeline|ImpactWeakPoint%20 PlayFromStart)
          (:Update
            (Effects|Components|ParticleSystem|SetFloatParameter _returnvalue "User.Point" _newtrack_0)
            (Effects|Components|ParticleSystem|SetFloatParameter _returnvalue "User.Point_Light" _newtrack_0)
            (Utilities|FlowControl|Sequence
              (:then_0
                _)
              (:then_1
                (Utilities|FlowControl|RetriggerableDelay 0.225)
                (bind (_direction_2 _newtrack_1) (Timeline|Timeline_1 PlayFromStart)
                  (:Update
                    (Effects|Components|ParticleSystem|SetFloatParameter _returnvalue "User.Point_Light" _newtrack_1))))
              (:then_2
                (Utilities|FlowControl|RetriggerableDelay 0.4)))
            (Utilities|FlowControl|RetriggerableDelay 0.225)
            (bind (_direction_3 _newtrack_1_2) (Timeline|Timeline_0 PlayFromStart)
              (:Update
                (Effects|Components|ParticleSystem|SetFloatParameter _returnvalue "User.Point" _newtrack_1_2)))))))))

(event EventBeginPlay)

(event EventTick
  (params (DeltaSeconds :type Float%20%28single-precision%29)))

(fn ConstructionScript
  (|CallParentFunction))
